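# Niagara ribbon trail for the CephaloChef emote, named as the fish trail.
# A single TrailMain emitter spawns very short-lived particles at a steady rate, taking color and ribbon width from the FishColor and RibbonWidth user parameters.
# Curves fade alpha and taper width to zero over each particle's life.
# RibbonWidth defaults to zero, so the ribbon has no width unless that parameter is overridden.
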NiagaraSystem /BRCosmetics/Animation/Game/MainPlayer/Emotes/CephaloChef/FX/NS_Emote_CephaloChef_FishTrail
dependencies {usedDataInterfaces [&NiagaraDataInterfaceCurve], usedDynamicInputs [], usedModules [&SystemState]}
emitters {TrailMain {
  emitterUpdateScript [
   &EmitterState()
   &SpawnRate(SpawnRate=150, Use Spawn Probability=false)
  ]
  particleSpawnScript [&InitializeParticle(
    Color=$User.FishColor
    Lifetime=0.08
    Ribbon Width=$User.RibbonWidth
    Ribbon Width Mode=::NewEnumerator1
    UsePositionOffset=false
    Write Material Random=true
   )]
  particleUpdateScript [
   &ParticleState()
   &ScaleColor(
    Scale Alpha=&FloatFromCurve(FloatCurve=DataInterface() {propertyValues {Curve {keys [<at 0 1>,<at 1 0>]}}})
    ScaleA=true
    ScaleRGB=true
   )
   &SolveForcesAndVelocity(Clamp Velocity=false, Limit Acceleration=false)
   &ScaleRibbonWidth(Ribbon Width Scale=&FloatFromCurve(FloatCurve=DataInterface() {propertyValues {Curve {keys [<at 0 1 RCIM_Cubic>, {arriveTangent=-2.8783, interpMode=RCIM_Cubic, leaveTangent=-2.8783, tangentMode=RCTM_User, time=1, value=0}]}}}))
  ]
  rendererClasses [&NiagaraRibbonRendererProperties]
  renderers [{rendererClass=&NiagaraRibbonRendererProperties}]
 }}
summary {
 systemName=NS_Emote_CephaloChef_FishTrail
 emitters [{emitterName=TrailMain, rendererClasses [&NiagaraRibbonRendererProperties]}]
 userVariables [
  User.FishColor {type=LinearColor, description="", defaultValue {struct=LinearColor, value=<rgba 1 1 1 1>}}
  User.RibbonWidth {type=NiagaraFloat, defaultValue=0, description=""}
 ]
}
