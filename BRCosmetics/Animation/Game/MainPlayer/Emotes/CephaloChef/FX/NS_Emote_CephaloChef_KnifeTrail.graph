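# Niagara system that draws a short-lived, pale blue knife trail for the CephaloChef emote.
# Its single emitter spawns FN anim-trail particles from bone or socket indices, with a ribbon ID offset utility and two ribbon renderers.
# Particles inherit source position movement and update as an anim trail.
# Curves fade the trail over its life, with color and intensity adjusted by time of day.
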NiagaraSystem /BRCosmetics/Animation/Game/MainPlayer/Emotes/CephaloChef/FX/NS_Emote_CephaloChef_KnifeTrail
dependencies {
 usedDataInterfaces [&NiagaraDataInterfaceCurve, &NiagaraDataInterfaceSocketReader]
 usedDynamicInputs []
 usedModules [
  &self:SystemSpawnScript.SystemScriptSource.SystemScriptGraph.NiagaraNodeAssignment_0.SetVariables_#1
  &SystemState
 ]
}
emitters {FN_AnimTrail {
  emitterSpawnScript [&FN_RibbonIDOffsetUtility()]
  emitterUpdateScript [
   &EmitterState()
   &FN_SpawnAnimTrail()
  ]
  particleSpawnScript [
   &FN_InitializeAnimTrailBoneIndecies()
   &FN_InitializeAnimTrailBoneIndecies(AnimTrail Selection=1, First Bone or Socket Index=2, Second Bone or Socket Index=3) {enabled=false, module=FN_InitializeAnimTrailBoneIndecies001}
   &FN_InitializeAnimTrail(Color=<rgba 0.5221 0.5926 0.7913 0.5>, Lifetime=0.2)
   &AddVelocity(Velocity=<0 400 0>) {enabled=false}
  ]
  particleUpdateScript [
   &InheritSourceMovement(ApplyForce=false, ApplyPosition=true, ApplyPositionFalloff=false)
   &SolveForcesAndVelocity(Clamp Velocity=false, Limit Acceleration=false)
   &FN_UpdateAnimTrail() {module=FN_UpdateAnimTrail001}
   &ScaleColor(
    Scale Alpha=&FloatFromCurve(
     FloatCurve=DataInterface() {propertyValues {Curve {keys [{arriveTangent=-1.622, interpMode=RCIM_Cubic, leaveTangent=-1.622, tangentMode=RCTM_User, time=0, value=1}, <at 1 0 RCIM_Cubic>]}}}
     Scale Curve=0.125
    )
    ScaleA=true
    ScaleRGB=true
   ) {enabled=false}
   &ScaleColor(
    Scale Alpha=&FloatFromCurve(FloatCurve=DataInterface() {propertyValues {Curve {keys [
         {arriveTangent=-21.3684, interpMode=RCIM_Cubic, leaveTangent=-21.3684, tangentMode=RCTM_User, time=0.1, value=1}
         {arriveTangent=-6.7025, interpMode=RCIM_Cubic, leaveTangent=-6.7025, tangentMode=RCTM_User, time=0.1531, value=0.3765}
         <at 0.3 0 RCIM_Cubic>
        ]}}})
    ScaleA=true
    ScaleRGB=true
   ) {module=ScaleColor001}
   &ScaleColor(
    Scale Alpha=&FloatFromCurve(
     FloatCurve=DataInterface() {propertyValues {Curve {keys [
         {arriveTangent=-40.3311, interpMode=RCIM_Cubic, leaveTangent=-40.3311, tangentMode=RCTM_User, time=0, value=2.5}
         {arriveTangent=-1.3327, interpMode=RCIM_Cubic, leaveTangent=-1.3327, tangentMode=RCTM_User, time=0.1, value=1.25}
         {interpMode=RCIM_Cubic, tangentMode=RCTM_User, time=1, value=0}
        ]}}}
     Scale Curve=&FN_Select_Float_from_Time_Of_Day(Morning=0.5, Day=0.4, Evening=0.5, Night=0.45)
    )
    Scale RGB=&MakeVectorFromLinearColorRGB(Color=&FNTimeOfDayColor(
      Night Color=<rgba 0.2 0.275 0.5 1>
      Morning Color=<rgba 0.5 0.45 0.4542 1>
      Day Color=<rgba 0.7 0.7 0.7 1>
      Evening Color=<rgba 0.65 0.4333 0.39 1>
     ))
    ScaleA=true
    ScaleRGB=true
   ) {module=ScaleColor002}
  ]
  rendererClasses [&NiagaraRibbonRendererProperties]
  renderers [
   {rendererClass=&NiagaraRibbonRendererProperties}
   {rendererClass=&NiagaraRibbonRendererProperties, rendererIndex=1}
  ]
 }}
summary {
 systemName=NS_Emote_CephaloChef_KnifeTrail
 emitters [{emitterName=FN_AnimTrail, rendererClasses [&NiagaraRibbonRendererProperties]}]
 userVariables [User.RibbonIDOffset {type=NiagaraInt32}]
}
