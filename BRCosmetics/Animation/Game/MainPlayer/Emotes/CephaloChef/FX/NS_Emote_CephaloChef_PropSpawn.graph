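# Niagara effect for the CephaloChef emote's prop spawn, layering airglow, sparkles, toon impact smoke and low/medium toon smoke emitters.
# The plain ToonSmoke emitter is present but disabled.
# A system spawn assignment module accompanies SystemState, with time-of-day color and float-from-array selection among the dynamic inputs used.
# It exposes a NormalizedNotifyProgress user float, which by its name tracks animation notify progress.
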NiagaraSystem /BRCosmetics/Animation/Game/MainPlayer/Emotes/CephaloChef/FX/NS_Emote_CephaloChef_PropSpawn
dependencies {
 usedDataInterfaces [&NiagaraDataInterfaceArrayFloat]
 usedDynamicInputs [&FNTimeOfDayColor, &SelectFloatFromArray]
 usedModules [
  &SystemState
  &self:SystemSpawnScript.SystemScriptSource.SystemScriptGraph.NiagaraNodeAssignment_0.SetVariables_#1
 ]
}
emitters {
 AirglowLWE001 {rendererClasses [], renderers []}
 SparklesLWE001 {rendererClasses [], renderers []}
 ToonImpactSmoke {rendererClasses [], renderers []}
 ToonSmoke {bEnabled=false, rendererClasses [], renderers []}
 ToonSmoke_LowMedium {rendererClasses [], renderers []}
}
summary {
 systemName=NS_Emote_CephaloChef_PropSpawn
 emitters [
  {emitterName=AirglowLWE001, rendererClasses []}
  {emitterName=SparklesLWE001, rendererClasses []}
  {emitterName=ToonImpactSmoke, rendererClasses []}
  {bEnabled=false, emitterName=ToonSmoke, rendererClasses []}
  {emitterName=ToonSmoke_LowMedium, rendererClasses []}
 ]
 userVariables [User.NormalizedNotifyProgress {type=NiagaraFloat, defaultValue=0, description=""}]
}
