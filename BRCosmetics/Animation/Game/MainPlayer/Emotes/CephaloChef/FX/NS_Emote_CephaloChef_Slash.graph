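# Niagara effect for the slash in the CephaloChef emote, combining a short-lived cut with flying fish bits.
# MainSlash and Directional_Streaks are sprites driven by User RotationAngle and Velocity, layering extra renderers and custom attributes for dark and blurry copies.
# FishGuts bursts a few pink or orange meshes along User FishVelocity, which fall, drag and shrink.
# Time-of-day scalars adjust slash and streak brightness.
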NiagaraSystem /BRCosmetics/Animation/Game/MainPlayer/Emotes/CephaloChef/FX/NS_Emote_CephaloChef_Slash
dependencies {
 usedDataInterfaces [
  &NiagaraDataInterfaceColorCurve
  &NiagaraDataInterfaceCurve
  &NiagaraDataInterfaceVector2DCurve
  &NiagaraDataInterfaceArrayColor
 ]
 usedDynamicInputs [&Vector4FromCurve]
 usedModules [&SystemState]
}
emitters {
 Directional_Streaks {
  emitterUpdateScript [
   &EmitterState()
   &SpawnBurst_Instantaneous(Spawn Time=0.02, Use Spawn Probability=true, UseLoopCountLimit=false)
  ]
  particleSpawnScript [
   &InitializeParticle(
    Lifetime=0.15
    Sprite Rotation Angle=$User.RotationAngle
    Sprite Rotation Mode=::NewEnumerator0
    Sprite Size Max=<5 65>
    Sprite Size Min=<3 40>
    Sprite Size Mode=::NewEnumerator2
    UsePositionOffset=false
    Write Material Random=true
   ) {module=InitializeParticle001}
   &SubUVAnimation(End Frame=3, SubUV Animation Mode:ENiagara_SubUVLookupMode=::NewEnumerator2)
   &AddVelocity(Velocity=$User.Velocity, Velocity Speed Scale=2.65)
   &CameraOffset()
   /SetVariables(Particles.Size Blurry Streak:Vector2f=<20 150>) {isSetParameters=true}
   /SetVariables(Particles.Color Dark Streak:LinearColor=<rgba 0.2344 0.2344 0.2344 0.5>, Particles.Size Dark Streak:Vector2f=<5 100>) {isSetParameters=true}
  ]
  particleUpdateScript [
   &ParticleState()
   &SolveForcesAndVelocity(Clamp Velocity=false, Limit Acceleration=false)
   &ScaleSpriteSize(Scale Factor=&Vector2DFromCurve(
     Vector2Curve=DataInterface() {propertyValues {XCurve {keys [<at 0 1 RCIM_Cubic>,<at 1 0 RCIM_Cubic>]}, YCurve {keys [<at 0 0.5 RCIM_Cubic>,<at 1 2 RCIM_Cubic>]}}}
     ScaleCurve=&Multiply_Vector2D_ByFloat(Float=1, Vector2D=<1.5 1>)
    ))
   &ScaleColor(
    Scale Alpha=&FloatFromCurve(FloatCurve=DataInterface() {propertyValues {Curve {keys [<at 0 1 RCIM_Cubic>,<at 1 0>]}}}, Scale Curve=0.2)
    Scale RGB=&VectorFromFloat(Value=&FloatFromCurve(
      FloatCurve=DataInterface() {propertyValues {Curve {keys [<at 0 10 RCIM_Cubic>,<at 1 1 RCIM_Cubic>]}}}
      Scale Curve=&Multiply_Float(
       A=0.5
       B=&FNTimeOfDayScalar(Night Scale=0.25, Morning Scale=0.45, Day Scale=0.8, Evening Scale=0.45)
      )
     ))
    ScaleA=true
    ScaleRGB=true
   )
   /SetVariables(
    Particles.Color Dark Streak:LinearColor=&MakeLinearColorFromVectorAndFloat(
     "Vector (RGB)"=&MakeVectorFromLinearColorRGB(Color=$Particles.Initial.Color Dark Streak)
     "Float (Alpha)"=&FloatFromCurve(FloatCurve=DataInterface() {propertyValues {Curve {keys [{interpMode=RCIM_Cubic, tangentMode=RCTM_User, time=0, value=1}, {interpMode=RCIM_Cubic, tangentMode=RCTM_User, time=0.5, value=0}]}}})
    )
    Particles.Size Dark Streak:Vector2f=&Multiply_Vector2D(
     A=$Particles.Initial.Size Dark Streak
     B=&MakeVector2D(
      X=1
      Y=&FloatFromCurve(
       FloatCurve=DataInterface() {propertyValues {Curve {keys [
           {arriveTangent=1.3254, interpMode=RCIM_Cubic, leaveTangent=1.3254, tangentMode=RCTM_User, time=0, value=0}
           {interpMode=RCIM_Cubic, tangentMode=RCTM_User, time=1, value=1}
          ]}}}
       Scale Curve=0.5
      )
     )
    )
   ) {isSetParameters=true}
  ]
  rendererClasses [&NiagaraSpriteRendererProperties]
  renderers [
   {rendererClass=&NiagaraSpriteRendererProperties}
   {rendererClass=&NiagaraSpriteRendererProperties, rendererIndex=1}
   {rendererClass=&NiagaraSpriteRendererProperties, rendererIndex=2}
  ]
 }
 FishGuts {
  emitterUpdateScript [
   &EmitterState()
   &SpawnBurst_Instantaneous(
    Spawn Count=&RandomRangeInt(Minimum=4, Maximum=7)
    Use Spawn Probability=false
    UseLoopCountLimit=false
   )
  ]
  particleSpawnScript [
   &InitializeParticle(
    Color=&SelectLinearColorFromArray(
     Array Sampling Mode=::NewEnumerator0
     Direct Array Index=?
     Color Selection Array=DataInterface() {propertyValues {ColorData [<rgba 1 0.1485 0.1963 1>, <rgba 1 0.4798 0.0845 1>]}}
     Random Seed=0
    )
    Lifetime Max=0.4
    Lifetime Min=0.1
    Lifetime Mode=::NewEnumerator1
    Mesh Scale Mode=::NewEnumerator1
    Mesh Uniform Scale Max=0.35
    Mesh Uniform Scale Min=0.55
    UsePositionOffset=false
    Write Material Random=true
   )
   &InitialMeshOrientation(Use Rotation Vector=false)
   &AddVelocity(Velocity=&Multiply_VectorByFloat(Vector=$User.FishVelocity, Float=&RandomRangeFloat(Minimum=0.2, Maximum=1.2)))
  ]
  particleUpdateScript [
   &ParticleState()
   &ScaleMeshSize(Scale Factor=&VectorFromFloat(Value=&FloatFromCurve(FloatCurve=DataInterface() {propertyValues {Curve {defaultValue=1, keys [<at 0 1 RCIM_Cubic>, {interpMode=RCIM_Cubic, leaveTangent=-8.0089, time=1, value=0}]}}})))
   &GravityForce(Gravity=&MakeVector(
     X=0
     Y=0
     Z=&FloatFromCurve(
      FloatCurve=DataInterface() {propertyValues {Curve {keys [<at 0 0>,<at 0.1 1>]}}}
      Scale Curve=&RandomRangeFloat(Minimum=-500, Maximum=-1150)
     )
    ))
   &Drag(Drag=&RandomRangeFloat(Minimum=1.8, Maximum=2.5), Use Linear Drag=true, Use Rotational Drag=true)
   &SolveForcesAndVelocity(Clamp Velocity=false, Limit Acceleration=false)
  ]
  rendererClasses [&NiagaraMeshRendererProperties]
  renderers [{rendererClass=&NiagaraMeshRendererProperties, rendererIndex=0}]
 }
 MainSlash {
  emitterUpdateScript [
   &EmitterState(Life Cycle Mode=::NewEnumerator1, Loop Behavior=::NewEnumerator1, Loop Duration=1, UseLoopDelay=false)
   &SpawnBurst_Instantaneous(Spawn Count=1, Use Spawn Probability=true)
  ]
  particleSpawnScript [
   &InitializeParticle(
    Color=<rgba 1.9 1.9 1.9 1>
    Lifetime=0.225
    Position Offset=$System.StartOffset
    Sprite Rotation Angle=$User.RotationAngle
    Sprite Rotation Mode=::NewEnumerator0
    Sprite Size Max=<5 175>
    Sprite Size Min=<3 100>
    Sprite Size Mode=::NewEnumerator2
    UsePositionOffset=false
    Write Material Random=true
   ) {module=InitializeParticle001}
   &AddVelocity(Velocity=$User.Velocity, Velocity Speed Scale=4.25)
   &CameraOffset(Camera Offset Amount=20)
   &DynamicMaterialParameters(
    Index 0 Param 1=1
    Index 0 Param 2=1
    Module.Param0WriteEnabled=true
    Module.Param1WriteEnabled=true
    Module.Param2WriteEnabled=true
    Module.Param3WriteEnabled=true
   )
  ]
  particleUpdateScript [
   &ParticleState()
   &SolveForcesAndVelocity(Clamp Velocity=false, Limit Acceleration=false)
   &ScaleColor(
    Color Value To Scale=$Particles.Color
    Scale Alpha=&FloatFromCurve(FloatCurve=DataInterface() {propertyValues {Curve {keys [
         {interpMode=RCIM_Cubic, tangentMode=RCTM_User, time=0, value=1}
         {arriveTangent=-3.4409, interpMode=RCIM_Cubic, leaveTangent=-3.4409, tangentMode=RCTM_User, time=1, value=0}
        ]}}})
    Scale RGB=&Multiply_VectorByFloat(
     Vector=&MakeVectorFromLinearColorRGB(Color=&ColorFromCurve(ColorCurve=DataInterface() {propertyValues {
         AlphaCurve {keys [<at 0 1 RCIM_Cubic>]}
         BlueCurve {keys [<at 0 0.5>,<at 0.2843 1.25 RCIM_Cubic>,<at 0.5346 0.5 RCIM_Cubic>]}
         GreenCurve {keys [<at 0 0.5>,<at 0.2843 1.25 RCIM_Cubic>,<at 0.5346 0.5 RCIM_Cubic>]}
         RedCurve {keys [<at 0 0.5>,<at 0.2843 1.25 RCIM_Cubic>,<at 0.5346 0.5 RCIM_Cubic>]}
        }}))
     Float=&FNTimeOfDayScalar(Night Scale=0.7, Morning Scale=0.7, Day Scale=0.7, Evening Scale=0.7)
    )
    ScaleA=true
    ScaleRGB=true
   )
   &ScaleSpriteSize(
    Non-Uniform Curve Sprite Scale=DataInterface() {propertyValues {
      XCurve {keys [<at 0 0 RCIM_Cubic>, {arriveTangent=-0.1, interpMode=RCIM_Cubic, leaveTangent=-0.1, time=0.2, value=1.075}, <at 1 -0.1 RCIM_Cubic>]}
      YCurve {keys [
        <at 0 0 RCIM_Cubic>
        {arriveTangent=0.0607, interpMode=RCIM_Cubic, leaveTangent=0.0607, tangentMode=RCTM_User, time=0.1, value=1.025}
        <at 1 0.8 RCIM_Cubic>
       ]}
     }}
    Scale Sprite Size Mode=::NewEnumerator3
   ) {module=ScaleSpriteSize001}
   &FN_Sprite_Attributes(
    Color=<rgba 0 0 0 1>
    Create Color Attribute=true
    Create Sprite Rotation Attribute=true
    Create SpriteSize Attribute=true
    SpriteRotation:NiagaraFloat=$User.RotationAngle
    SpriteSize=&Multiply_Vector2D(A=$Particles.SpriteSize, B=<1.1 1.25>)
    Write to Module Namespace=true
   )
  ]
  rendererClasses [&NiagaraSpriteRendererProperties]
  renderers [
   {rendererClass=&NiagaraSpriteRendererProperties}
   {rendererClass=&NiagaraSpriteRendererProperties, rendererIndex=1}
  ]
 }
}
summary {
 systemName=NS_Emote_CephaloChef_Slash
 emitters [
  {emitterName=MainSlash, rendererClasses [&NiagaraSpriteRendererProperties]}
  {emitterName=Directional_Streaks, rendererClasses [&NiagaraSpriteRendererProperties]}
  {emitterName=FishGuts, rendererClasses [&NiagaraMeshRendererProperties]}
 ]
 userVariables [
  User.bIsFrontEnd {type=NiagaraBool}
  User.bIsFrontEndPreview {type=NiagaraBool}
  User.FishVelocity {type=Vector3f, description="", defaultValue {struct=Vector3f, value=<0 0 0>}}
  User.RotationAngle {type=NiagaraFloat, defaultValue=0, description=""}
  User.Velocity {type=Vector3f, description="", defaultValue {struct=Vector3f, value=<0 0 0>}}
 ]
}
